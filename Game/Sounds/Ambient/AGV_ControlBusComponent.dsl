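; Blueprint component for Audio Gameplay Volumes that drives an Audio Modulation control bus as the volume toggles.
; When the volume toggles on, it sets the global control bus mix value of its Control Bus variable to full over the configurable Fade Time.
; When it toggles off, it fades that bus back to zero.
; Both changes apply only while the component is active.
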(event AudioGameplayVolume|EventOnToggledOn
  (if (Components|Activation|IsActive)
    (Audio|Modulation|SetGlobalControlBusMixValue (Variables|Default|GetControlBus) 1.0 (Variables|Default|GetFade%20Time))))

(event AudioGameplayVolume|EventOnToggledOff
  (if (Components|Activation|IsActive)
    (Audio|Modulation|SetGlobalControlBusMixValue (Variables|Default|GetControlBus) 0.0 (Variables|Default|GetFade%20Time))))
