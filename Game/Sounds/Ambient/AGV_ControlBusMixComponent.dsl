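; AGV_ControlBusMixComponent is an Audio Gameplay Volume component Blueprint that ties an audio modulation Control Bus Mix to the volume's toggle state.
; When the volume toggles on, it activates the mix held in its ControlBusMix variable.
; When the volume toggles off, it deactivates the same mix.
; BeginPlay and Tick are empty.
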(event EventBeginPlay)

(event EventTick
  (params (DeltaSeconds :type Float%20%28single-precision%29)))

(event AudioGameplayVolume|EventOnToggledOn
  (Audio|Modulation|ActivateControlBusMix (Variables|Default|GetControlBusMix)))

(event AudioGameplayVolume|EventOnToggledOff
  (Audio|Modulation|DeactivateControlBusMix (Variables|Default|GetControlBusMix)))
