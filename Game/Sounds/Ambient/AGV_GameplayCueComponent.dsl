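; Audio gameplay volume component that applies looping gameplay cues to the locally viewed Athena player pawn during selected day phases.
; Toggle on/off events track whether the volume is entered, while the Time of Day Manager's phase-change event rechecks validity and updates cues if inside.
; UpdateGameplayCue adds or removes each configured tag client-side through local looping cue calls, with TagApplied tracking when removal is needed.
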(event EventBeginPlay
  (bind _timeofdaymanager (Class|FortAsyncActionTimeOfDayManagerReady|GetTimeOfDayManagerReady)
    (:OnReady
      (TimeOfDay|RegisterforDayPhaseChangedEvent _timeofdaymanager (EventDispatchers|CreateEvent :!event Custom|OnDayPhaseChange) true))))

(event Custom|OnDayPhaseChange
  (params (CurrentDayPhase :type EFortDayPhase%20Enum :default "Morning") (PreviousDayPhase :type EFortDayPhase%20Enum :default "Morning") (bAtCreation :type Boolean))
  ;[1 1 1] Update time of day when player enter space or time of day changes.
  (bind _output_get (Variables|DayPhase|SetCurrent%20Day%20Phase CurrentDayPhase))
  (Variables|Default|SetIsDayPhaseValid (Utilities|Array|ContainsItem (Variables|DayPhase|GetDayPhaseToAdd) _output_get))
  (if (Variables|Default|GetEnteredSpace)
    (CallFunction|EnteredPhaseUpdate)))

(event AudioGameplayVolume|EventOnToggledOn
  (bind _gameplay_cue_container_tags (Variables|TagstoApply|GetGameplay%20Cue%20Container%20Tags))
  (bind _gameplaytags (GameplayTags|BreakGameplayTagContainer _gameplay_cue_container_tags))
  ;[1 1 1] Update gameplay tag that is applied to player based on enter or leaving the volume.
  (Variables|Default|SetEnteredSpace true)
  (Variables|Default|SetIsDayPhaseValid (Utilities|Array|ContainsItem (Variables|DayPhase|GetDayPhaseToAdd) (Variables|DayPhase|GetCurrent%20Day%20Phase)))
  (if (Variables|Default|GetIsDayPhaseValid)
    (for _array_element _gameplaytags
      (Utilities|IsA(soft) (Pawn|GetViewTarget (Fortnite|GetFirstLocalFortPlayerController true true)) "/Script/FortniteGame.FortPlayerPawnAthena"
        (:True
          (CallFunction|UpdateGameplayCue true _array_element))))
    (Variables|Default|SetTagApplied (GameplayTags|IsGameplayTagValid (Utilities|Array|Get(acopy) _gameplaytags)))))

(event Custom|EnteredPhaseUpdate
  (bind _returnvalue (Fortnite|GetFirstLocalFortPlayerController true true))
  (bind _returnvalue_2 (Pawn|GetViewTarget _returnvalue))
  (bind _gameplay_cue_container_tags (Variables|TagstoApply|GetGameplay%20Cue%20Container%20Tags))
  (bind _gameplaytags (GameplayTags|BreakGameplayTagContainer _gameplay_cue_container_tags))
  ;[1 1 1] Update gameplay tag that is applied to player based on enter or leaving the volume.
  (if (Variables|Default|GetIsDayPhaseValid)
    (if (Variables|Default|GetIsDayPhaseValid)
      (for _array_element _gameplaytags
        (Utilities|IsA(soft) _returnvalue_2 "/Script/FortniteGame.FortPlayerPawnAthena"
          (:True
            (CallFunction|UpdateGameplayCue true _array_element))))
      (Variables|Default|SetTagApplied (GameplayTags|IsGameplayTagValid (Utilities|Array|Get(acopy) _gameplaytags))))
    (elif (Variables|Default|GetTagApplied)
      (for _array_element_2 _gameplaytags
        (Utilities|IsA(soft) _returnvalue_2 "/Script/FortniteGame.FortPlayerPawnAthena"
          (:True
            (CallFunction|UpdateGameplayCue false _array_element_2))))
      (Variables|Default|SetTagApplied false))))

(event AudioGameplayVolume|EventOnToggledOff
  ;[1 1 1] Update gameplay tag that is applied to player based on enter or leaving the volume.
  (Variables|Default|SetEnteredSpace false)
  (if (Variables|Default|GetTagApplied)
    (for _array_element (GameplayTags|BreakGameplayTagContainer (Variables|TagstoApply|GetGameplay%20Cue%20Container%20Tags))
      (Utilities|IsA(soft) (Pawn|GetViewTarget (Fortnite|GetFirstLocalFortPlayerController true true)) "/Script/FortniteGame.FortPlayerPawnAthena"
        (:True
          (CallFunction|UpdateGameplayCue false _array_element))))
    (Variables|Default|SetTagApplied false)))

(fn UpdateGameplayCue
  (params (AddGameplayCue :type Boolean) (GameplayCueTag :type Gameplay%20Tag%20Structure))
  (bind _returnvalue (Fortnite|GetFirstLocalFortPlayerController true true))
  (bind _returnvalue_2 (Pawn|GetViewTarget _returnvalue))
  ;[1 1 1] Applies or remove gameplay tag on local player controller. This is client based. Notice this is setup for looping. Could be upgrade for future to do enter and exit burst tags too.
  (if AddGameplayCue
    (Projectile|AddGameplayCue%28Looping%29Local _returnvalue_2 GameplayCueTag (Ability|GameplayCue|MakeGameplayCueParameters :NormalizedMagnitude 1.0 :RawMagnitude 1.0 :Location "0,0,0" :Normal "0,0,0" :Instigator _returnvalue :GameplayEffectLevel 1 :AbilityLevel 1 :bReplicateLocationWhenUsingMinimalRepProxy false))
    (else
      (Projectile|RemoveGameplayCue%28Looping%29Local _returnvalue_2 GameplayCueTag (Ability|GameplayCue|MakeGameplayCueParameters 0.0)))))
