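; B_AmbientSound_RandomRep is an ambient sound actor built to play its sound at random intervals.
; Once playlist data is ready, the authority starts a timer with a delay drawn from a random time range and, if attached, binds to the parent's destruction to destroy itself.
; The timer fires PlayAmbientSound, which re-rolls the timer on authority and sets and plays the audio component only on remotes.
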(event EventBeginPlay
  (Class|FortAsyncActionCurrentPlaylistReady|PlaylistDataReadyAsync
    (:OnReady
      (Networking|SwitchHasAuthority
        (:Authority
          (CallFunction|ResetTimer)
          (Utilities|IsValid (Actor|GetAttachParentActor)
            (:Is%20Valid
              (Game|BindEventtoOnDestroyed (delegate Custom|AttachActorDestroyed)))))))))

(event Custom|ResetTimer
  (bind (_randomtimerange_x _randomtimerange_y) (Math|Vector2D|BreakVector2D (Variables|GetRandomTimeRange_X)))
  (Utilities|Time|SetTimerbyEvent (EventDispatchers|CreateEvent :!event Custom|PlayAmbientSound) (Math|Random|RandomFloatinRange _randomtimerange_x _randomtimerange_y)))

(event Custom|AttachActorDestroyed
  (params (DestroyedActor :type Actor%20Object%20Reference))
  (Actor|DestroyActor))

(event Custom|PlayAmbientSound
  (Networking|SwitchHasAuthority
    (:Authority
      (CallFunction|ResetTimer))
    (:Remote
      (Audio|Components|Audio|SetSound (Variables|Default|GetAudio) (Variables|Default|GetSound))
      (Audio|Components|Audio|Play (Variables|Default|GetAudio) 0.0))))

(fn ConstructionScript)
